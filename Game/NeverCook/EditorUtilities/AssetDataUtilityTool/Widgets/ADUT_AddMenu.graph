EditorUtilityWidgetBlueprint /Game/NeverCook/EditorUtilities/AssetDataUtilityTool/Widgets/ADUT_AddMenu
graphs [
 CleanUpList {nodes [
   CleanUpList() {is=title K2Node_FunctionEntry}
   > clearItems(self:TA Generic 01 List Object Reference=Get TA_Generic01_List() -> (TA_Generic01_List:TA Generic 01 List Object Reference))
  ]}
 EventGraph {nodes [
   "On Clicked (TA_Generic01_Button)"!() -> (OutputDelegate, then, widget:User Widget Object Reference, event:Pointer Event Structure) {is=title K2Node_ComponentBoundEvent}
   > SetVisibility(InVisibility=Collapsed)
   "On Clicked (btnCreate)"!() -> (OutputDelegate, then, widget:User Widget Object Reference, event:Pointer Event Structure) {is=title K2Node_ComponentBoundEvent}
   > HandleCreate()
  ]}
 FinalizeCreate {nodes [
   n1 = FinalizeCreate!() -> (then, class:Object Class Reference) {is=title K2Node_FunctionEntry}
   > Call On Create Happened!(execute:Exec, self:ADUT Add Menu Object Reference, class:Object Class Reference=@n1) -> (then:Exec) {is=title K2Node_CallDelegate}
   > SetVisibility(InVisibility=Collapsed)
  ]}
 HandleCreate {nodes [
   HandleCreate() {is=title K2Node_FunctionEntry}
   > n1 = getSelectedData!(execute, self:TA Generic 01 List Object Reference=Get TA_Generic01_List()) -> (then, selectedData:Array of Object References)
   > Branch(Condition=Is Valid Index!(TargetArray:Array of Object References=@n1, IndexToTest:int) -> (ReturnValue:bool))
   > n2 = Cast To TA_GenericListItemData!(execute, Object="Get (a copy)"(Array:Array of Object References=@n1) -> (Output:Object Reference)) -> (then, CastFailed, AsTA Generic List Item Data:TA Generic List Item Data Object Reference) {is=title K2Node_DynamicCast}
   n3 = Get strings!(self:TA Generic List Item Data Object Reference=@n2) -> (strings:Array of Strings)
   > Branch(Condition=Is Valid Index(TargetArray:Array of Strings=@n3))
   n4 = "Get (a copy)"(Array:Array of Strings=@n3) -> (Output:String)
   > n5 = GetAssetsByPackageName!(
    execute
    self:Asset Registry Interface=GetAssetRegistry()
    PackageName:Name=String To Name(InString=@n4)
    bIncludeOnlyOnDiskAssets:bool=false
    bSkipARFilteredAssets:bool=true
   ) -> (then, OutAssetData:Array of Asset Data Structures, ReturnValue:bool)
   > n6 = Branch(Condition=Is Valid Index(TargetArray:Array of Asset Data Structures=@n5))
   > n7 = LoadBlueprintClass!(
    execute
    AssetPath:String="To String (Name)"(InName="Get (a copy)"!(Array:Array of Asset Data Structures=@n5, Dimension 1) -> (Output_PackageName:Name, Output_PackagePath:Name, Output_AssetName:Name, Output_AssetClassPath:Top Level Asset Path Structure))
   ) -> (then, ReturnValue:Object Class Reference)
   > Branch(Condition=IsValidClass!(Class:Object Class Reference=@n7) -> (ReturnValue:bool))
   > FinalizeCreate!(execute, self, class:Object Class Reference=@n7) -> (then) {is=title K2Node_CallFunction}
   n8 = Append!(A=@n4, B=., C:String="Get (a copy)"!(Array:Array of Strings=@n3, Dimension 1=1) -> (Output:String)) -> (ReturnValue)
   n9 = LoadClassAsset_Blocking(execute=@n6.else, AssetClass=To Soft Class Reference(SoftClassPath=MakeSoftClassPath(PathString=@n8)))
   > Branch(Condition=IsValidClass(Class=@n9))
   > FinalizeCreate(class=@n9)
  ]}
 ShowMenu {nodes [
   n1 = ShowMenu!() -> (then, classes:Array of Top Level Asset Path Structures) {is=title K2Node_FunctionEntry}
   > CleanUpList() {is=title K2Node_CallFunction}
   > n2 = For Each Loop(Array:Array of Top Level Asset Path Structures=@n1) -> (Array Element:Top Level Asset Path Structure)
   SetVisibility(execute=@n2.Completed, InVisibility=SelfHitTestInvisible)
   Set l_path(execute=@n2.LoopBody, l_path:Top Level Asset Path Structure=@n2) -> (Output_Get:Top Level Asset Path Structure)
   n3 = Break Top Level Asset Path!(TopLevelAssetPath:"Top Level Asset Path Structure (by ref)"=@n2) -> (PackageName, AssetName) {is=title K2Node_BreakStruct}
   n4 = "To String (Name)"(InName=@n3.AssetName)
   > n5 = Branch(Condition=StartsWith(SourceString=@n4, InPrefix=SKEL_, SearchCase=CaseSensitive))
   n6 = addItemGeneric!(
    execute=@n5.else
    self:TA Generic 01 List Object Reference=Get TA_Generic01_List()
    displayString:String="To String (Name)"(InName=@n3.AssetName)
   ) -> (then, widget:User Widget Object Reference, data:Object Reference)
   > n7 = Cast To TA_GenericListItemData(Object=@n6.data)
   > Set strings!(
    execute
    strings:Array of Strings=Make Array!("[0]":String="To String (Name)"(InName=@n3), "[1]":String=@n4) -> (Array:Array of Strings)
    self:TA Generic List Item Data Object Reference=@n7
   ) -> (then, Output_Get:Array of Strings)
   n8 = Break Top Level Asset Path(TopLevelAssetPath=Get l_path() -> (l_path:Top Level Asset Path Structure))
   > n9 = Branch(Condition=StartsWith(SourceString="To String (Name)"(InName=@n8), InPrefix=/Script))
   n10 = GetAssetsByPackageName(
    execute=@n9.else
    self=GetAssetRegistry()
    PackageName=Break Top Level Asset Path(TopLevelAssetPath=Get l_path())
    bIncludeOnlyOnDiskAssets=false
    bSkipARFilteredAssets=true
   )
   > Branch(Condition=Is Valid Index(TargetArray=@n10))
   n11 = MakeLiteralName(Value=BlueprintDescription)
   n12 = "Get (a copy)"(Array:Array of Asset Data Structures=@n10) -> (Output:Asset Data Structure)
   > Set ToolTipText!(execute, ToolTipText:Text="To Text (String)"(InString=GetTagValue(InAssetData=@n12, InTagName=@n11)), self:Widget Object Reference=@n6) -> (then, Output_Get:Text)
  ]}
 onCreateHappened {nodes [onCreateHappened!() -> (then, class:Object Class Reference) {is=title K2Node_FunctionEntry}]}
]
variables [onCreateHappened {default=()}]
widgets {
 info {rootWidgetClass=&VerticalBox, widgetCount=11}
 widgets [
  VerticalBox_1 {widgetClassPath=&VerticalBox, widget=VerticalBox_1}
  Border_2 {
   widgetClassPath=&Border
   parent=VerticalBox_1
   widget=Border_2
   properties {brushColor=<rgba 0 0 0 0.49>, padding=<ltrb 0 0 0 0>}
   slot {class=VerticalBoxSlot, properties {size {sizeRule=Fill, value=1}}}
  }
  BackgroundBlur_39 {
   widgetClassPath=&BackgroundBlur
   parent=Border_2
   widget=BackgroundBlur_39
   properties {blurStrength=7.2, padding=<ltrb 50 50 50 50>}
   slot {class=BorderSlot, properties {padding=<ltrb 0 0 0 0>}}
  }
  Border_0 {
   widgetClassPath=&Border
   parent=BackgroundBlur_39
   widget=Border_0
   properties {brushColor=<rgba 0.0176 0.0176 0.0176 1>}
   slot {class=BackgroundBlurSlot, properties {padding=<ltrb 50 50 50 50>}}
  }
  VerticalBox_3 {widgetClassPath=&VerticalBox, parent=Border_0, widget=VerticalBox_3, slot {class=BorderSlot}}
  Border_1 {
   widgetClassPath=&Border
   parent=VerticalBox_3
   widget=Border_1
   properties {brushColor=<rgba 0.0284 0.0284 0.0284 1>}
   slot {class=VerticalBoxSlot}
  }
  HorizontalBox_0 {widgetClassPath=&HorizontalBox, parent=Border_1, widget=HorizontalBox_0, slot {class=BorderSlot}}
  txtTitle {
   widgetClassPath=&CommonTextBlock
   parent=HorizontalBox_0
   widget=txtTitle
   properties {
    text=Add New
    font {
     bForceMonospaced=false
     bMaterialIsStencil=false
     fontMaterial=null
     fontObject=&NotoSans
     letterSpacing=0
     monospacedWidth=1
     size=12
     skewAmount=0
     typefaceFontName=NotoSans-Regular
     outlineSettings {
      bApplyOutlineToDropShadows=false
      bMiteredCorners=false
      bSeparateFillAlpha=false
      outlineMaterial=null
      outlineSize=0
      outlineColor=<rgba 0 0 0 1>
     }
    }
    shadowOffset=<0 0>
   }
   slot {class=HorizontalBoxSlot, properties {verticalAlignment=VAlign_Center, size {sizeRule=Fill, value=1}}}
  }
  TA_Generic01_Button {
   widgetClassPath=&TA_Generic01_Button
   parent=HorizontalBox_0
   widget=TA_Generic01_Button
   properties {
    buttonColor_Default=<rgba 1 0 0.026 1>
    iconBrush {
     bIsDynamicallyLoaded=false
     drawAs=Image
     imageType=NoImage
     mirroring=NoMirror
     resourceName=None
     resourceObject=/Game/UI/Icons/Icon_Seach_Close
     tiling=NoTile
     imageSize=<16 16>
     margin=<ltrb 0 0 0 0>
     outlineSettings {
      bUseBrushTransparency=false
      roundingType=HalfHeightRadius
      width=0
      color {colorUseRule=UseColor_Specified, specifiedColor=<rgba 0 0 0 0>}
      cornerRadii=<0 0 0 1>
     }
     tintColor {colorUseRule=UseColor_Specified, specifiedColor=<rgba 1 1 1 1>}
     uVRegion {bIsValid=false, max=<0 0>, min=<0 0>}
    }
   }
   slot {class=HorizontalBoxSlot, properties {verticalAlignment=VAlign_Center}}
  }
  TA_Generic01_List {
   widgetClassPath=/Game/Blueprints/Blutilities/ForTechArt/Widgets/List/TA_Generic01_List
   bIsVariable=true
   parent=VerticalBox_3
   widget=TA_Generic01_List
   slot {class=VerticalBoxSlot, properties {size {sizeRule=Fill, value=1}}}
  }
  btnCreate {
   widgetClassPath=&TA_Generic01_Button
   parent=VerticalBox_3
   widget=btnCreate
   properties {label=Create}
   slot {class=VerticalBoxSlot, properties {verticalAlignment=VAlign_Center}}
  }
 ]
}
