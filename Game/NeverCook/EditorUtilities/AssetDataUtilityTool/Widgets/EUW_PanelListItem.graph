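# Editor utility list item widget for the Asset Data Utility Tool that shows one object as a name label above its editable details view.
# When the list assigns an object, SetObject feeds the DetailsView and labels the item with the object path's text after the last colon.
# For non-/Script classes, it reads the BlueprintDescription tag from the Asset Registry, without loading other assets, as the label's tooltip.
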EditorUtilityWidgetBlueprint /Game/NeverCook/EditorUtilities/AssetDataUtilityTool/Widgets/EUW_PanelListItem
graphs [
 EventGraph {nodes [
   n1 = Event OnListItemObjectSet()
   > SetObject!(execute, self:Self Object Reference, object:Object Reference=@n1.ListItemObject) -> (then)
   Event PreConstruct()
   Event Construct()
   Event Tick()
  ]}
 SetObject {nodes [
   n1 = SetObject!() -> (then, object:Object Reference) {is=title K2Node_FunctionEntry}
   > Set currentObject(currentObject:Object Reference=@n1) -> (Output_Get:Object Reference)
   > Is Valid(InputObject=@n1)
   > SetObject!(
    execute
    self:Property View Base Object Reference=Get DetailsView_18() -> (DetailsView_18:Details View Object Reference)
    NewObject:Object Reference=Get currentObject() -> (currentObject:Object Reference)
   ) -> (then) {is=title K2Node_CallFunction}
   n2 = GetSoftObjectPath!(Object=Get currentObject()) -> (ReturnValue_PathString:String)
   > "SetText (Text)"(
    self=Get txtToolName() -> (txtToolName:Common Text Object Reference)
    InText="To Text (String)"(InString=Split(SourceString=@n2, InStr=:, SearchDir=FromEnd).RightS)
   )
   > n3 = Sequence()
   n4 = GetSoftClassPath!(Class:Object Class Reference=Get Class(Object=Get currentObject()) -> (ReturnValue:Object Class Reference)) -> (ReturnValue_PathString:String)
   > n5 = Branch(Condition=NOT Boolean(A=StartsWith(SourceString=@n4, InPrefix=/Script)))
   > n6 = GetAssetsByPackageName!(
    execute
    self:Asset Registry Interface=GetAssetRegistry()
    PackageName:Name=String To Name(InString=Split(SourceString=@n4, InStr=., SearchDir=FromStart))
    bIncludeOnlyOnDiskAssets:bool=false
    bSkipARFilteredAssets:bool=true
   ) -> (then, OutAssetData:Array of Asset Data Structures, ReturnValue:bool)
   > Branch(Condition=Is Valid Index!(TargetArray:Array of Asset Data Structures=@n6, IndexToTest:int) -> (ReturnValue:bool))
   > Set l_tooltip(l_tooltip:String=GetTagValue(
     InAssetData="Get (a copy)"(Array:Array of Asset Data Structures=@n6) -> (Output:Asset Data Structure)
     InTagName=MakeLiteralName(Value=BlueprintDescription)
    )) -> (Output_Get:String)
   Set l_tooltip(execute=@n5.else)
   Set ToolTipText!(
    execute=@n3.then_1
    ToolTipText:Text="To Text (String)"(InString=Get l_tooltip() -> (l_tooltip:String))
    self:Widget Object Reference=Get txtToolName()
   ) -> (then, Output_Get:Text)
  ]}
]
variables [currentObject {type=Object, default=null}]
widgets {
 info {rootWidgetClass=&Border, widgetCount=4}
 widgets [
  Border_0 {widgetClassPath=&Border, widget=Border_0, properties {brushColor=<rgba 0.0075 0.0075 0.0075 1>}}
  VerticalBox_0 {widgetClassPath=&VerticalBox, parent=Border_0, widget=VerticalBox_0, slot {class=BorderSlot}}
  txtToolName {
   widgetClassPath=&CommonTextBlock
   bIsVariable=true
   parent=VerticalBox_0
   widget=txtToolName
   properties {
    visibility=Visible
    font {
     bForceMonospaced=false
     bMaterialIsStencil=false
     fontMaterial=null
     fontObject=&NotoSans
     letterSpacing=0
     monospacedWidth=1
     size=12
     skewAmount=0
     typefaceFontName=NotoSans-Regular
     outlineSettings {
      bApplyOutlineToDropShadows=false
      bMiteredCorners=false
      bSeparateFillAlpha=false
      outlineMaterial=null
      outlineSize=0
      outlineColor=<rgba 0 0 0 1>
     }
    }
    shadowOffset=<0 0>
   }
   slot {class=VerticalBoxSlot}
  }
  DetailsView_18 {
   widgetClassPath=&DetailsView
   parent=VerticalBox_0
   widget=DetailsView_18
   slot {class=VerticalBoxSlot, properties {padding=<ltrb 15 0 0 0>}}
  }
 ]
}
